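# Player animation blueprint for the anti-vehicle mounted turret, posing a Fortnite character manning it through the vehicle final pose override layer.
# Armored Battle Bus turret passenger idle, via DefaultSlot, gets an additive lean from smoothed yaw delta and an aim offset from pitch.
# Hand IK bones snap to world-space attach points, then Leg IK nodes solve both arms using per-hand alphas before inertialization.
# The input final pose is unused, and other layers contain no logic.
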AnimBlueprint /ResolveTurret/Items/Environmental/AntiVehicleMountedTurret/Animations/AVMT_Player_AnimBP
target_skeleton=&Fortnite_M_Avg_Player_Skeleton
graphs [
 AnimGraph {nodes [Output Pose() {layerGroup=None}]}
 EventGraph {links [], nodes []}
 VehicleFinalPoseOverride {nodes [
   Input Pose() {name=InFinalPose}
   n1 = Get YawDeltaSmoothed() -> (YawDeltaSmoothed:float)
   n2 = Sequence Player 'BattleBusVehicle_TurretPassenger_Idle'() {sequence=/Game/Animation/Game/MainPlayer/Vehicles/ArmoredBattleBus/BattleBusVehicle_TurretPassenger_Idle}
   n3 = AimOffset Player 'BattleBusVehicle_TurretPassenger_AO'!(
    X:float=Get AimPitch()
    BasePose:Pose Link Structure=Apply Additive(
     Base=Slot 'DefaultSlot'(Source=@n2)
     Additive=Blendspace Player 'BattleBusVehicle_TurretPassenger_Lean_BS'!(X=@n1) -> (Pose) {blendSpace=/Game/Animation/Game/MainPlayer/Vehicles/ArmoredBattleBus/BattleBusVehicle_TurretPassenger_Lean_BS}
    ) {alphaBoolBlend {blendOption=Linear}}
    Alpha:float=1
   ) -> (Pose:Pose Link Structure) {
    blendSpace=/Game/Animation/Game/MainPlayer/Vehicles/ArmoredBattleBus/BattleBusVehicle_TurretPassenger_AO
    alphaBoolBlend {blendOption=Linear}
   }
   n4 = "Transform (Modify) Bone - Bone: ik_hand_l"!(ComponentPose=Local To Component(LocalPose=@n3), Alpha=1, Translation=Get HandAttachL() -> (HandAttachL:Vector)) -> (Pose) {translationMode=BMM_Replace, translationSpace=BCS_WorldSpace, alphaBoolBlend {blendOption=Linear}, boneToModify {boneName=ik_hand_l}}
   n5 = "Transform (Modify) Bone - Bone: ik_hand_r"!(ComponentPose=@n4, Alpha=1, Translation=Get HandAttachR() -> (HandAttachR:Vector)) -> (Pose) {translationMode=BMM_Replace, translationSpace=BCS_WorldSpace, alphaBoolBlend {blendOption=Linear}, boneToModify {boneName=ik_hand_r}}
   n6 = Leg IK!(ComponentPose:Component Space Pose Link Structure=@n5, Alpha:float=Get RightHandIKAlpha() -> (RightHandIKAlpha:float)) -> (Pose:Component Space Pose Link Structure) {
    is=title AnimGraphNode_LegIK
    alphaBoolBlend {blendOption=Linear}
    legsDefinition [{
      bEnableKneeTwistCorrection=false
      bEnableRotationLimit=false
      footBoneForwardAxis=Y
      hingeRotationAxis=Z
      minRotationAngle=15
      numBonesInLimb=2
      twistOffsetCurveName=None
      fKFootBone {boneName=hand_r}
      iKFootBone {boneName=ik_hand_r}
     }]
   }
   n7 = Leg IK(ComponentPose=@n6, Alpha=Get LeftHandIKAlpha() -> (LeftHandIKAlpha:float)) {
    alphaBoolBlend {blendOption=Linear}
    legsDefinition [{
      bEnableKneeTwistCorrection=false
      bEnableRotationLimit=false
      footBoneForwardAxis=Y
      hingeRotationAxis=Z
      minRotationAngle=15
      numBonesInLimb=2
      twistOffsetCurveName=None
      fKFootBone {boneName=hand_l}
      iKFootBone {boneName=ik_hand_l}
     }]
   }
   Output Pose(Result=Inertialization!(Source:Pose Link Structure=Component To Local(ComponentPose=@n7)) -> (Pose:Pose Link Structure) {is=title AnimGraphNode_Inertialization}) {layerGroup=VehicleOverride}
  ]}
 VehicleFullBodyOverride {nodes [
   Input Pose() {name=InPoseFullBody}
   Output Pose() {layerGroup=VehicleOverride}
  ]}
 VehicleLowerBodyOverride {nodes [
   Input Pose() {name=InPoseLowerBodyDefault}
   Input Pose() {name=InPoseUpperBody}
   Output Pose() {layerGroup=VehicleOverride}
  ]}
 VehicleSplitBodyOverride {nodes [
   Input Pose() {name=InPoseSplitBody}
   Input Pose() {name=InPoseUpperAndLowerBody}
   Input Pose() {name=InPoseUpperAndLowerBodyRemoveRoot}
   Output Pose() {layerGroup=Vehicle Override}
  ]}
]
variables []
