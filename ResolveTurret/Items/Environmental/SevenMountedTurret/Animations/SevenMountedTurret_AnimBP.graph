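# Animation Blueprint for the Seven mounted turret, posing it by aim and driver presence.
# Blends default or off pose by HasDriver, passes it through DefaultSlot, then applies pitch and yaw-delta lean aim offsets.
# The base bone's world-space rotation is then replaced using AimingYaw, followed by inertialization.
# Each update refreshes HasDriver and copies the driver's AVMT anim layer yaw delta into YawDeltaSmoothed_old, which the AnimGraph does not read.
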AnimBlueprint /ResolveTurret/Items/Environmental/SevenMountedTurret/Animations/SevenMountedTurret_AnimBP
target_skeleton=/ResolveTurret/Items/Environmental/SevenMountedTurret/Meshes/Anti_Personnel_Mounted_Turret_Gadget_Skeleton
graphs [
 AnimGraph {nodes [
   n1 = Sequence Player 'SevenMountedTurret_AO_DefaultPose_Weapon'() {sequence=&/SevenMountedTurret_AO_DefaultPose_Weapon}
   n2 = Sequence Player 'SMT_Off_Pose_Weapon'() {sequence=&/SMT_Off_Pose_Weapon}
   n3 = Get HasDriver() -> (HasDriver:bool)
   n4 = AimOffset Player 'SevenMountedTurret_AO_Weapon'!(
    X:float=NormalizeAxis!(Angle:float=Get AimingPitch() -> (AimingPitch:float)) -> (ReturnValue:float)
    BasePose:Pose Link Structure=Slot 'DefaultSlot'(Source=Blend Poses by bool(BlendPose_0=@n1, BlendPose_1=@n2, bActiveValue=@n3) {blendType=Linear})
    Alpha:float=1
   ) -> (Pose:Pose Link Structure) {
    blendSpace=&/SevenMountedTurret_AO_Weapon
    alphaBoolBlend {blendOption=Linear}
   }
   n5 = AimOffset Player 'SevenMountedTurret_LeanAO'!(X:float=Get YawDeltaSmoothed() -> (YawDeltaSmoothed:float), BasePose:Pose Link Structure=@n4, Alpha:float=1) -> (Pose:Pose Link Structure) {blendSpace=&/SevenMountedTurret_LeanAO, alphaBoolBlend {blendOption=Linear}}
   n6 = "Transform (Modify) Bone - Bone: base"!(
    ComponentPose=Local To Component(LocalPose=@n5)
    Alpha=1
    Rotation_Roll:float=0
    Rotation_Pitch:float=0
    Rotation_Yaw:float=Get AimingYaw() -> (AimingYaw:float)
   ) -> (Pose) {rotationMode=BMM_Replace, rotationSpace=BCS_WorldSpace, alphaBoolBlend {blendOption=Linear}, boneToModify {boneName=base}}
   Output Pose(Result=Inertialization!(Source:Pose Link Structure=Component To Local(ComponentPose=@n6)) -> (Pose:Pose Link Structure) {is=title AnimGraphNode_Inertialization}) {layerGroup=None}
  ]}
 EventGraph {nodes [
   Event BlueprintUpdateAnimation()
   n1 = TryGetPawnOwner()
   > Is Valid(InputObject=@n1)
   > n2 = Cast To SevenMountedTurret!(execute, Object=@n1) -> (then, CastFailed, AsSeven Mounted Turret:Seven Mounted Turret Object Reference) {is=title K2Node_DynamicCast}
   > n3 = Set Driver Pawn(Driver Pawn:Fort Player Pawn Object Reference=Get DriverPawn!(self:Mounted Turret Object Reference=@n2) -> (DriverPawn:Fort Player Pawn Object Reference)) -> (Output_Get:Fort Player Pawn Object Reference)
   > n4 = Is Valid(InputObject=@n3)
   Set HasDriver(execute=@n4.Is Not Valid, HasDriver:bool=false) -> (Output_Get:bool)
   n5 = Cast To AVMT_Player_AnimBP!(
    execute=@n4.Is Valid
    Object=GetLinkedAnimLayerInstanceByClass!(
     self:Skeletal Mesh Component Object Reference=GetComponentByClass!(self:Actor Object Reference=@n3, ComponentClass:Actor Component Class Reference=/Script/Engine.SkeletalMeshComponent) -> (ReturnValue:Skeletal Mesh Component Object Reference)
     InClass:Anim Instance Class Reference=/ResolveTurret/Items/Environmental/AntiVehicleMountedTurret/Animations/AVMT_Player_AnimBP.AVMT_Player_AnimBP_C
    ) -> (ReturnValue:Anim Instance Object Reference)
   ) -> (then, CastFailed, AsAVMT Player Anim BP:AVMT Player Anim BP Object Reference) {is=title K2Node_DynamicCast}
   > Set YawDeltaSmoothed_old(YawDeltaSmoothed_old:double=Get YawDeltaSmoothed!(self:Fort Player Anim Instance Proxy Object Reference=@n5) -> (YawDeltaSmoothed)) -> (Output_Get:double)
   > Set HasDriver(HasDriver=IsValid(Object=Get Driver Pawn() -> (Driver Pawn:Fort Player Pawn Object Reference)))
  ]}
]
variables [
 Driver Pawn {type=FortPlayerPawn}
 HasDriver {type=boolean}
 YawDeltaSmoothed_old {type=number}
]
