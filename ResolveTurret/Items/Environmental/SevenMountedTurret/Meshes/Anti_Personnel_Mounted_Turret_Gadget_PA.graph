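# Physics asset for the Anti-Personnel Mounted Turret gadget mesh in ResolveTurret, giving its skeleton simple collision bodies.
# Each bone from the offset and base through the three-segment gun arm, shield, and left and right gun bodies and barrels carries one box.
# Constraints follow that chain, branching at the arm end to the shield and both guns, all with uniform swing and twist limits.
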PhysicsAsset /ResolveTurret/Items/Environmental/SevenMountedTurret/Meshes/Anti_Personnel_Mounted_Turret_Gadget_PA
bodies {
 base [{shapeType=Box, extentX=114.5912, extentY=69.7058, extentZ=10.3377, center=<-33.5519 0 7.2059>}]
 gun_arm_end [{shapeType=Box, extentX=52.6226, extentY=13.5323, extentZ=37.8906, center=<-19.3399 0 12.047>}]
 gun_arm_mid [{shapeType=Box, extentX=37.5256, extentY=15.4815, extentZ=42.5006, center=<10.8015 0 13.2644>}]
 gun_arm_start [{shapeType=Box, extentX=31.3083, extentY=22.1569, extentZ=46.927, center=<5.6728 0 15.6429>}]
 gun_barrel_01_l [{shapeType=Box, extentX=59.5216, extentY=13.0665, extentZ=20.1487, center=<28.6701 -0.4879 -0.2386>}]
 gun_barrel_01_r [{shapeType=Box, extentX=59.5216, extentY=13.0664, extentZ=20.1487, center=<28.6701 -0.4879 0.2386>}]
 gun_body_l [{shapeType=Box, extentX=113.7193, extentY=50.5719, extentZ=49.7779, center=<20.4564 -7.6241 -0.3195>}]
 gun_body_r [{shapeType=Box, extentX=113.7193, extentY=50.5719, extentZ=49.7779, center=<20.4564 -7.6241 0.3195>}]
 gun_shield [{shapeType=Box, extentX=43.9526, extentY=75.9325, extentZ=56.2242, center=<-16.6044 0 -24.4899>}]
 offset [{shapeType=Box, extentX=124.8104, extentY=124.8104, extentZ=9.8956, center=<0 0 4.8988>}]
}
constraints [
 {bone1Name=base, bone2Name=offset, swing1LimitDegrees=45, swing2LimitDegrees=45, twistLimitDegrees=45}
 {bone1Name=gun_arm_start, bone2Name=base, swing1LimitDegrees=45, swing2LimitDegrees=45, twistLimitDegrees=45}
 {bone1Name=gun_arm_mid, bone2Name=gun_arm_start, swing1LimitDegrees=45, swing2LimitDegrees=45, twistLimitDegrees=45}
 {bone1Name=gun_arm_end, bone2Name=gun_arm_mid, swing1LimitDegrees=45, swing2LimitDegrees=45, twistLimitDegrees=45}
 {bone1Name=gun_shield, bone2Name=gun_arm_end, swing1LimitDegrees=45, swing2LimitDegrees=45, twistLimitDegrees=45}
 {bone1Name=gun_body_r, bone2Name=gun_arm_end, swing1LimitDegrees=45, swing2LimitDegrees=45, twistLimitDegrees=45}
 {bone1Name=gun_barrel_01_r, bone2Name=gun_body_r, swing1LimitDegrees=45, swing2LimitDegrees=45, twistLimitDegrees=45}
 {bone1Name=gun_body_l, bone2Name=gun_arm_end, swing1LimitDegrees=45, swing2LimitDegrees=45, twistLimitDegrees=45}
 {bone1Name=gun_barrel_01_l, bone2Name=gun_body_l, swing1LimitDegrees=45, swing2LimitDegrees=45, twistLimitDegrees=45}
]
